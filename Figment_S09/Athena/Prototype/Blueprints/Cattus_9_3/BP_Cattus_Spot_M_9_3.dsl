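; Manager for Cattus spot events that schedules timed appearances of a spot prop during a calendar event.
; On the mesh network root, a data curve over event progress scales trigger delay, with random variance, and shore distance.
; Each trigger rolls position, duration and vocals, which the remote branch passes to the prop.
; Tuning values and next UTC trigger time live in mesh metadata, editable through RootSet events and restored when read.
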(event Custom|PlaySpotEvent
  (params (Position :type Float%20%28double-precision%29) (Duration :type Float%20%28double-precision%29) (ShoreDistance :type Float%20%28double-precision%29) (Vocals :type Boolean))
  (Networking|SwitchHasAuthority
    (:Remote
      (Class|BPCattusSpot93|SetParametersandStartEvent (Variables|Default|GetProp) Position Duration (* ShoreDistance -1.0) Vocals)
      (Development|PrintString (Utilities|String|ToString(Float) Position) true true "(R=0.000000,G=0.660000,B=1.000000,A=1.000000)" 52.0)
      (Development|PrintString (Utilities|String|ToString(Float) Duration) true true "(R=1.000000,G=0.342181,B=0.462278,A=1.000000)" 52.0)
      (Development|PrintString (Utilities|String|ToString(Float) ShoreDistance) true true "(R=0.481878,G=1.000000,B=0.264462,A=1.000000)" 52.0))))

(event Custom|RootSetTriggerTimeVariance
  (params (MinDelayBetweenTriggers :type Float%20%28double-precision%29))
  (bind _output_get (Variables|Default|SetTriggerTimeVarianceMultiplier MinDelayBetweenTriggers))
  (bind _structout (Utilities|Struct|SetmembersinCattusSpotStruct (Variables|Default|GetCattSpotMetaData) _output_get))
  (MeshNetworkSubsystem|SetMetaDatawithKey (GameInstanceSubsystems|GetMeshNetworkSubsystem) (Variables|Default|GetMetadataKey) _structout))

(event Custom|RootSetMaxDelayBetweenTriggers
  (params (MinDelayBetweenTriggers :type Float%20%28double-precision%29))
  (bind _output_get (Variables|Default|SetMaxDelayBetweenTriggers MinDelayBetweenTriggers))
  (bind _structout (Utilities|Struct|SetmembersinCattusSpotStruct (Variables|Default|GetCattSpotMetaData) _output_get))
  (MeshNetworkSubsystem|SetMetaDatawithKey (GameInstanceSubsystems|GetMeshNetworkSubsystem) (Variables|Default|GetMetadataKey) _structout))

(event Custom|RootSetMinDelayBetweenTriggers
  (params (MinDelayBetweenTriggers :type Float%20%28double-precision%29))
  (bind _output_get (Variables|Default|SetMinDelayBetweenTriggers MinDelayBetweenTriggers))
  (bind _structout (Utilities|Struct|SetmembersinCattusSpotStruct (Variables|Default|GetCattSpotMetaData) _output_get))
  (MeshNetworkSubsystem|SetMetaDatawithKey (GameInstanceSubsystems|GetMeshNetworkSubsystem) (Variables|Default|GetMetadataKey) _structout))

(event Custom|RootSetDurationMin
  (params (DurationMin :type Float%20%28double-precision%29))
  (bind _output_get (Variables|Default|SetDurationMin DurationMin))
  (bind _structout (Utilities|Struct|SetmembersinCattusSpotStruct (Variables|Default|GetCattSpotMetaData) _output_get))
  (MeshNetworkSubsystem|SetMetaDatawithKey (GameInstanceSubsystems|GetMeshNetworkSubsystem) (Variables|Default|GetMetadataKey) _structout))

(event Custom|RootSetDurationMax
  (params (DurationMax :type Float%20%28double-precision%29))
  (bind _output_get (Variables|Default|SetDurationMax DurationMax))
  (bind _structout (Utilities|Struct|SetmembersinCattusSpotStruct (Variables|Default|GetCattSpotMetaData) _output_get))
  (MeshNetworkSubsystem|SetMetaDatawithKey (GameInstanceSubsystems|GetMeshNetworkSubsystem) (Variables|Default|GetMetadataKey) _structout))

(event Custom|RootSetShoreDistanceMin
  (params (ShoreDistanceMin :type Float%20%28double-precision%29))
  (bind _output_get (Variables|Default|SetShoreDistanceMin ShoreDistanceMin))
  (bind _structout (Utilities|Struct|SetmembersinCattusSpotStruct (Variables|Default|GetCattSpotMetaData) _output_get))
  (MeshNetworkSubsystem|SetMetaDatawithKey (GameInstanceSubsystems|GetMeshNetworkSubsystem) (Variables|Default|GetMetadataKey) _structout))

(event Custom|RootSetShoreDistanceMax
  (params (ShoreDistanceMax :type Float%20%28double-precision%29))
  (bind _output_get (Variables|Default|SetShoreDistanceMax ShoreDistanceMax))
  (bind _structout (Utilities|Struct|SetmembersinCattusSpotStruct (Variables|Default|GetCattSpotMetaData) _output_get))
  (MeshNetworkSubsystem|SetMetaDatawithKey (GameInstanceSubsystems|GetMeshNetworkSubsystem) (Variables|Default|GetMetadataKey) _structout))

(event Custom|RootSetChanceForVocals
  (params (ChanceForVocals :type Float%20%28double-precision%29))
  (bind _output_get (Variables|Default|SetChanceForVocals ChanceForVocals))
  (bind _structout (Utilities|Struct|SetmembersinCattusSpotStruct (Variables|Default|GetCattSpotMetaData) _output_get))
  (MeshNetworkSubsystem|SetMetaDatawithKey (GameInstanceSubsystems|GetMeshNetworkSubsystem) (Variables|Default|GetMetadataKey) _structout))

(event Custom|RootSetNextTriggerUTCTime
  (params (inMinutes :type Float%20%28double-precision%29))
  (bind _output_get (Variables|Default|SetNextTriggerUTCTime (Math|DateTime|DateTime+Timespan (Math|DateTime|UTCNow) (Math|Timespan|FromMinutes inMinutes))))
  (bind _structout (Utilities|Struct|SetmembersinCattusSpotStruct (Variables|Default|GetCattSpotMetaData) _output_get))
  (MeshNetworkSubsystem|SetMetaDatawithKey (GameInstanceSubsystems|GetMeshNetworkSubsystem) (Variables|Default|GetMetadataKey) _structout)
  (CallFunction|RootCallEvent))

(event Custom|RootCallEvent
  (bind _output_get (Variables|Default|SetCurrentUTCTime (Math|DateTime|UTCNow)))
  (Utilities|Time|ClearandInvalidateTimerbyHandle (Variables|Default|GetNextTriggerTimer))
  (bind _output_get_2 (Variables|Default|SetNextTimer (Math|Timespan|GetTotalSeconds (Math|DateTime|DateTime-DateTime (Variables|Default|GetNextTriggerUTCTime) _output_get))))
  (bind _returnvalue (Utilities|Time|SetTimerbyEvent (EventDispatchers|CreateEvent :!event Custom|TriggerHandler) (Math|Float|Max(Float) _output_get_2 0.1)))
  (Variables|Default|SetNextTriggerTimer _returnvalue))

(event Collision|EventActorBeginOverlap
  (params (OtherActor :type Actor%20Object%20Reference)))

(event EventTick
  (params (DeltaSeconds :type Float%20%28single-precision%29)))

(event Custom|TriggerHandler
  ;[1 1 1] Clear the timer and play the spot event, then set the next timer
  (CallFunction|SetCurveValue)
  (Utilities|Time|ClearandInvalidateTimerbyHandle (Variables|Default|GetNextTriggerTimer))
  (CallFunction|PlaySpotEvent (Math|Random|RandomFloatinRange (Variables|Default|GetPositionMin) (Variables|Default|GetPositionMax)) (Math|Random|RandomFloatinRange (Variables|Default|GetDurationMin) (Variables|Default|GetDurationMax)) (Math|Float|MapRangeClamped (Variables|Default|GetCurveValue) 0.0 1.0 (Variables|Default|GetShoreDistanceMax) (Variables|Default|GetShoreDistanceMin)) (Math|Random|RandomBoolwithWeight (Variables|Default|GetChanceForVocals)))
  (CallFunction|SetNextTimer))

(event Custom|SetCurveValue
  (bind (_biseventactive _timesincebegin _timeuntilend _timespanratio) (CalendarEvents|GetCalendarEventInformation (Variables|Default|GetCalendarEventName)))
  (if _biseventactive
    (Variables|Default|SetCurveValue (Math|Curves|GetFloatValue (Variables|Default|GetDataCurve) _timespanratio))))

(event Custom|MeshMetaDataUpdated
  (bind _cattspotmetadata (Variables|Default|GetCattSpotMetaData))
  (bind (_durationmin_10_31ec07284fe19e6154d9b9a78e921e68 _durationmax_13_a16ce174466fd4fc63c4239b1778c22c _shoredistancemin_15_d1bae98a47522756f1a9039cb876b284 _shoredistancemax_17_058250344230807f8f82de923ece2388 _mindelaybetweentriggers_23_c14ed4db41148311af7a929e25e02671 _maxdelaybetweentriggers_24_54ae4fd444b8ea61ecd275b7e373387f _triggertimevariance_26_be96019646b437ae3e7b449b80399fcd _nettriggerutctime_29_e0420e74455cf2e52c94b6a17e163a66 _chanceforvocals_32_d0dd6d2b4cd1ca1840889c87559b5e49) (Utilities|Struct|BreakCattusSpotStruct _cattspotmetadata))
  (bind _returnvalue (GameInstanceSubsystems|GetMeshNetworkSubsystem))
  (bind _metadatakey (Variables|Default|GetMetaDataKey))
  (bind _returnvalue_2 (MeshNetworkSubsystem|GetMetaDatawithKey _returnvalue _metadatakey _cattspotmetadata))
  (if _returnvalue_2
    (Variables|Default|SetDurationMin _durationmin_10_31ec07284fe19e6154d9b9a78e921e68)
    (Variables|Default|SetDurationMax _durationmax_13_a16ce174466fd4fc63c4239b1778c22c)
    (Variables|Default|SetShoreDistanceMin _shoredistancemin_15_d1bae98a47522756f1a9039cb876b284)
    (Variables|Default|SetShoreDistanceMax _shoredistancemax_17_058250344230807f8f82de923ece2388)
    (Variables|Default|SetMinDelayBetweenTriggers _mindelaybetweentriggers_23_c14ed4db41148311af7a929e25e02671)
    (Variables|Default|SetMaxDelayBetweenTriggers _maxdelaybetweentriggers_24_54ae4fd444b8ea61ecd275b7e373387f)
    (Variables|Default|SetTriggerTimeVarianceMultiplier _triggertimevariance_26_be96019646b437ae3e7b449b80399fcd)
    (Variables|Default|SetChanceForVocals _chanceforvocals_32_d0dd6d2b4cd1ca1840889c87559b5e49)
    (Variables|Default|SetNextTriggerUTCTime _nettriggerutctime_29_e0420e74455cf2e52c94b6a17e163a66)
    (bind _biseventactive (CalendarEvents|GetCalendarEventInformation (Variables|Default|GetCalendarEventName)))
    (if _biseventactive
      (CallFunction|RootCallEvent))
    (else
      (MeshNetworkSubsystem|SetMetaDatawithKey _returnvalue _metadatakey _cattspotmetadata))))

(event Custom|SetNextTimer
  (bind (_biseventactive _timesincebegin _timeuntilend _timespanratio) (CalendarEvents|GetCalendarEventInformation (Variables|Default|GetCalendarEventName)))
  (if _biseventactive
    (CallFunction|CatSpotTimerSet _timespanratio)))

(event Custom|CatSpotTimerSet
  (params (TimeSpanRatio :type Float%20%28double-precision%29))
  (bind _output_get (Variables|Default|SetCurveValue (Math|Curves|GetFloatValue (Variables|Default|GetDataCurve) TimeSpanRatio)))
  (if (not (Utilities|Time|IsTimerActivebyHandle (Variables|Default|GetNextTriggerTimer)))
    (bind _output_get_2 (Variables|Default|SetNextTimer (+ (* (Math|Float|MapRangeClamped _output_get 0.0 1.0 (Variables|Default|GetMaxDelayBetweenTriggers) (Variables|Default|GetMinDelayBetweenTriggers)) 6e+01) (Math|Random|RandomFloatinRange (* (* (* (Math|Float|MapRangeClamped _output_get 0.0 1.0 (Variables|Default|GetMaxDelayBetweenTriggers) (Variables|Default|GetMinDelayBetweenTriggers)) 6e+01) (Variables|Default|GetTriggerTimeVarianceMultiplier)) -1.0) (* (* (Math|Float|MapRangeClamped _output_get 0.0 1.0 (Variables|Default|GetMaxDelayBetweenTriggers) (Variables|Default|GetMinDelayBetweenTriggers)) 6e+01) (Variables|Default|GetTriggerTimeVarianceMultiplier))))))
    (bind _returnvalue (Utilities|Time|SetTimerbyEvent (EventDispatchers|CreateEvent :!event Custom|TriggerHandler) _output_get_2))
    (Variables|Default|SetNextTriggerTimer _returnvalue)
    (bind _output_get_3 (Variables|Default|SetNextTriggerUTCTime (Math|DateTime|DateTime+Timespan (Variables|Default|GetCurrentUTCTime) (Math|Timespan|FromSeconds (Variables|Default|GetNextTimer)))))
    (bind _structout (Utilities|Struct|SetmembersinCattusSpotStruct (Variables|Default|GetCattSpotMetaData) _output_get_3))
    (MeshNetworkSubsystem|SetMetaDatawithKey (GameInstanceSubsystems|GetMeshNetworkSubsystem) (Variables|Default|GetMetadataKey) _structout)))

(event EventBeginPlay
  (bind _nodetype (Class|FortAsyncActionMeshNetworkReady|MeshNetworkReadyAsync)
    (:OnReady
      (switch Utilities|FlowControl|Switch|SwitchonEMeshNetworkNodeType _nodetype
        (:Root
          (Variables|Default|SetCurrentUTCTime (Math|DateTime|UTCNow))
          (CallFunction|MeshMetaDataUpdated)
          (Utilities|FlowControl|Sequence
            (:then_0
              (if (or (Math|DateTime|Equal(DateTime) (Variables|Default|GetNextTriggerUTCTime) (Variables|Default|GetDefaultUTCTime)) (Math|DateTime|DateTime>DateTime (Variables|Default|GetCurrentUTCTime) (Variables|Default|GetNextTriggerUTCTime)))
                (bind (_eventname _timeuntilend _timesincebegin _timespanratio) (MCPContext|WatchCalendarEvent (Variables|Default|GetCalendarEventName))
                  (:OnEventActive
                    (CallFunction|CatSpotTimerSet _timespanratio)))
                (elif (Utilities|Time|IsTimerActivebyHandle (Variables|Default|GetNextTriggerTimer))
                  _
                  (else
                    (bind _output_get (Variables|Default|SetNextTimer (Math|Timespan|GetTotalSeconds (Math|DateTime|DateTime-DateTime (Variables|Default|GetNextTriggerUTCTime) (Variables|Default|GetCurrentUTCTime)))))
                    (bind _returnvalue (Utilities|Time|SetTimerbyEvent (EventDispatchers|CreateEvent :!event Custom|TriggerHandler) _output_get))
                    (Variables|Default|SetNextTriggerTimer _returnvalue)))))
            (:then_1
              (MeshNetworkSubsystem|BindEventtoOnMeshMetaDataUpdated (GameInstanceSubsystems|GetMeshNetworkSubsystem) (EventDispatchers|CreateEvent :!event Custom|MeshMetaDataUpdated)))))))))

(fn OnRep_Progress
  (switch Utilities|FlowControl|Switch|SwitchonEMeshNetworkNodeType (MeshNetworkSubsystem|GetMeshNetworkNodeType (GameInstanceSubsystems|GetMeshNetworkSubsystem))
    (:Edge
      (Class|BPCattusSpot93|SetStartLocation (Variables|Default|GetProp) (Variables|Default|GetPosition))
      (Development|PrintString "rep triggered"))))

(fn ConstructionScript)
